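; Temporary preview Blueprint for the RyzaRevT2 vehicle cosmetic body that lays out its four wheels from axle settings.
; The Construction Script places each wheel from front or rear axle X and Z values, offset sideways by plus or minus half the axle width.
; It also multiplies each wheel's current relative scale by that axle's wheel scale.
; BeginPlay, BeginOverlap and Tick are disabled stubs with no logic.
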(event EventBeginPlay
  (|CallParentFunction) ; This node is disabled and will not be called.\nDrag off pins to build functionality.
)

(event Collision|EventActorBeginOverlap
  (params (OtherActor :type Actor%20Object%20Reference))
  (|CallParentFunction OtherActor) ; This node is disabled and will not be called.\nDrag off pins to build functionality.
)

(event EventTick
  (params (DeltaSeconds :type Float%20%28single-precision%29))
  (|CallParentFunction DeltaSeconds) ; This node is disabled and will not be called.\nDrag off pins to build functionality.
)

(fn ConstructionScript
  (bind _rearaxel_posx (Variables|Default|GetRearAxel_PosX))
  (bind _rearaxel_posz (Variables|Default|GetRearAxel_PosZ))
  (bind _rearaxel_width (Variables|Default|GetRearAxel_Width))
  (bind _rearaxel_wheelscale (Variables|Default|GetRearAxel_WheelScale))
  (bind _frontaxel_wheelscale (Variables|Default|GetFrontAxel_WheelScale))
  (bind _wheel_bl (Variables|Default|GetWheel_BL))
  (bind _wheel_br (Variables|Default|GetWheel_BR))
  (bind _frontaxel_posx (Variables|Default|GetFrontAxel_PosX))
  (bind _frontaxel_posz (Variables|Default|GetFrontAxel_PosZ))
  (bind _frontaxel_width (Variables|Default|GetFrontAxel_Width))
  (bind _wheel_fl (Variables|Default|GetWheel_FL))
  (bind _wheel_fr (Variables|Default|GetWheel_FR))
  (|CallParentFunction)
  (Transformation|SetRelativeLocation _wheel_fr (Math|Vector|MakeVector _frontaxel_posx (* _frontaxel_width 0.5) _frontaxel_posz) false true)
  (Transformation|SetRelativeScale3D _wheel_fr (* (Variables|/Script/Engine.SceneComponent|GetRelativeScale3D _wheel_fr) _frontaxel_wheelscale))
  (Transformation|SetRelativeLocation _wheel_fl (Math|Vector|MakeVector _frontaxel_posx (* _frontaxel_width -0.5) _frontaxel_posz) false true)
  (Transformation|SetRelativeScale3D _wheel_fl (* _frontaxel_wheelscale (Variables|/Script/Engine.SceneComponent|GetRelativeScale3D _wheel_fl)))
  (Transformation|SetRelativeLocation _wheel_br (Math|Vector|MakeVector _rearaxel_posx (* _rearaxel_width 0.5) _rearaxel_posz) false true)
  (Transformation|SetRelativeScale3D _wheel_br (* (Variables|/Script/Engine.SceneComponent|GetRelativeScale3D _wheel_br) _rearaxel_wheelscale))
  (Transformation|SetRelativeLocation _wheel_bl (Math|Vector|MakeVector _rearaxel_posx (* _rearaxel_width -0.5) _rearaxel_posz) false true)
  (Transformation|SetRelativeScale3D _wheel_bl (* _rearaxel_wheelscale (Variables|/Script/Engine.SceneComponent|GetRelativeScale3D _wheel_bl))))
